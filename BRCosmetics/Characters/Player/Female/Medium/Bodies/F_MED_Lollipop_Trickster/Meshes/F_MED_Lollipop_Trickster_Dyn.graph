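# Physics asset for the female medium Lollipop Trickster cosmetic body, covering its dynamic belt and strap bones.
# Pelvis, lower spine and thigh capsules form the body volumes, with belt boxes and strap capsules chained from the pelvis.
# Back straps split left and right from a middle chain, while front straps start at master bones.
# Angular-free constraints cross-link belt ends, front-to-back strap ends, and thighs to front straps, closing loops between chains.
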PhysicsAsset /BRCosmetics/Characters/Player/Female/Medium/Bodies/F_MED_Lollipop_Trickster/Meshes/F_MED_Lollipop_Trickster_Dyn
bodies {
 dyn_belt_l_01 [{shapeType=Box, extentX=3.6875, extentY=3.125, extentZ=0.6875, center=<2 0 0>}]
 dyn_belt_l_02 [{shapeType=Box, extentX=2.5, extentY=3.125, extentZ=0.6875, center=<1.6772 0 0>}]
 dyn_belt_l_03 [{shapeType=Box, extentX=1.875, extentY=3.125, extentZ=0.6875, center=<1.1354 0 0>}]
 dyn_belt_mid_01 [{shapeType=Capsule, length=1.3844, radius=0.6156, center=<1 0 0>, rotation=<rot 90 0 0>}]
 dyn_belt_mid_02 [{shapeType=Capsule, length=1.3844, radius=0.6156, center=<1 0 0>, rotation=<rot 90 0 0>}]
 dyn_belt_mid_MSTR [{shapeType=Capsule, length=0.5, radius=0.5}]
 dyn_belt_r_01 [{shapeType=Box, extentX=3.6875, extentY=3.125, extentZ=0.6875, center=<2 0 0>}]
 dyn_belt_r_02 [{shapeType=Box, extentX=2.5, extentY=3.125, extentZ=0.6875, center=<1.6772 0 0>}]
 dyn_belt_r_03 [{shapeType=Box, extentX=1.875, extentY=3.125, extentZ=0.6875, center=<1.1354 0 0>}]
 dyn_strap_bk_l_01 [{shapeType=Capsule, length=6.7631, radius=2.5, shapeName=dyn_strap_fr_l_05_capsule, center=<1.2871 0 1>, rotation=<rot 0 90 90>}]
 dyn_strap_bk_l_02 [{shapeType=Capsule, length=6.7631, radius=2.5, shapeName=dyn_strap_fr_l_05_capsule, center=<1.2871 0 1>, rotation=<rot 0 90 90>}]
 dyn_strap_bk_l_03 [{shapeType=Capsule, length=6.7631, radius=2.5, shapeName=dyn_strap_fr_l_05_capsule, center=<1.2871 0 1>, rotation=<rot 0 90 90>}]
 dyn_strap_bk_l_04 [{shapeType=Capsule, length=6.7631, radius=2.5, shapeName=dyn_strap_fr_l_05_capsule, center=<1.2871 0 1>, rotation=<rot 0 90 90>}]
 dyn_strap_bk_l_05 [{shapeType=Capsule, length=6.7631, radius=2.5, shapeName=dyn_strap_fr_l_05_capsule, center=<1.2871 0 1>, rotation=<rot 0 90 90>}]
 dyn_strap_bk_l_06 [{shapeType=Capsule, length=6.7631, radius=2.5, shapeName=dyn_strap_fr_l_05_capsule, center=<1.2871 0 1>, rotation=<rot 0 90 90>}]
 dyn_strap_bk_mid_01 [{shapeType=Capsule, length=0.1686, radius=2.8314, shapeName=dyn_strap_bk_mid_01_capsule, center=<3.5 0 2>, rotation=<rot 0 0 90>}]
 dyn_strap_bk_mid_02 [{shapeType=Capsule, length=0.0765, radius=2.9235, shapeName=dyn_strap_bk_mid_02_capsule, center=<2.5 0 2>, rotation=<rot 0 0 90>}]
 dyn_strap_bk_mid_03 [{shapeType=Capsule, length=1.2791, radius=2.7209, shapeName=dyn_strap_bk_mid_03_capsule, center=<2.5 0 2>, rotation=<rot 0 0 90>}]
 dyn_strap_bk_r_01 [{shapeType=Capsule, length=6.7631, radius=2.5, shapeName=dyn_strap_fr_l_05_capsule, center=<1.2871 0 1>, rotation=<rot 0 86.6218 -87.0721>}]
 dyn_strap_bk_r_02 [{shapeType=Capsule, length=6.7631, radius=2.5, shapeName=dyn_strap_fr_l_05_capsule, center=<1.2871 0 1>, rotation=<rot 0 86.6218 -87.0721>}]
 dyn_strap_bk_r_03 [{shapeType=Capsule, length=6.7631, radius=2.5, shapeName=dyn_strap_fr_l_05_capsule, center=<1.2871 0 1>, rotation=<rot 0 86.6218 -87.0721>}]
 dyn_strap_bk_r_04 [{shapeType=Capsule, length=6.7631, radius=2.5, shapeName=dyn_strap_fr_l_05_capsule, center=<1.2871 0 1>, rotation=<rot 0 86.6218 -87.0721>}]
 dyn_strap_bk_r_05 [{shapeType=Capsule, length=6.7631, radius=2.5, shapeName=dyn_strap_fr_l_05_capsule, center=<1.2871 0 1>, rotation=<rot 0 86.6218 -87.0721>}]
 dyn_strap_bk_r_06 [{shapeType=Capsule, length=6.7631, radius=2.5, shapeName=dyn_strap_fr_l_05_capsule, center=<1.2871 0 1>, rotation=<rot 0 86.6218 -87.0721>}]
 dyn_strap_fr_l_01 [{shapeType=Capsule, length=4.25, radius=2.5, center=<2.6599 0 1>, rotation=<rot 90 0 0>}]
 dyn_strap_fr_l_02 [{shapeType=Capsule, length=6.5131, radius=2.5, shapeName=dyn_strap_fr_l_05_capsule, center=<1.2871 0 1>, rotation=<rot 0 90 90>}]
 dyn_strap_fr_l_02a [{shapeType=Capsule, length=6.7631, radius=2.5, shapeName=dyn_strap_fr_l_05_capsule, center=<1.2871 0 1>, rotation=<rot 0 90 90>}]
 dyn_strap_fr_l_03 [{shapeType=Capsule, length=9.2631, radius=2.5, shapeName=dyn_strap_fr_l_05_capsule, center=<1.2871 0 1>, rotation=<rot 0 90 90>}]
 dyn_strap_fr_l_04 [{shapeType=Capsule, length=9.2631, radius=2.5, shapeName=dyn_strap_fr_l_05_capsule, center=<1.2871 0 1>, rotation=<rot 0 90 90>}]
 dyn_strap_fr_l_05 [{shapeType=Capsule, length=8.6381, radius=2.5, shapeName=dyn_strap_fr_l_05_capsule, center=<1.2871 0 1>, rotation=<rot 0 90 -90>}]
 dyn_strap_fr_l_06 [{shapeType=Capsule, length=6.5131, radius=2.5, shapeName=dyn_strap_fr_l_05_capsule, center=<1.2871 0 1>, rotation=<rot 0 90 -90>}]
 dyn_strap_fr_l_MSTR [{shapeType=Capsule, length=0.5, radius=0.5}]
 dyn_strap_fr_r_01 [{shapeType=Capsule, length=4.25, radius=2.5, center=<2.6599 0 1>, rotation=<rot 90 0 0>}]
 dyn_strap_fr_r_02 [{shapeType=Capsule, length=6.7631, radius=2.5, shapeName=dyn_strap_fr_l_05_capsule, center=<1.2871 0 1>, rotation=<rot 0 86.6218 -87.0721>}]
 dyn_strap_fr_r_02a [{shapeType=Capsule, length=6.7631, radius=2.5, shapeName=dyn_strap_fr_l_05_capsule, center=<1.2871 0 1>, rotation=<rot 0 86.6218 -87.0721>}]
 dyn_strap_fr_r_03 [{shapeType=Capsule, length=8.8881, radius=2.5, shapeName=dyn_strap_fr_l_05_capsule, center=<1.2871 0 1>, rotation=<rot 0 86.6218 -87.0721>}]
 dyn_strap_fr_r_04 [{shapeType=Capsule, length=8.8881, radius=2.5, shapeName=dyn_strap_fr_l_05_capsule, center=<1.2871 0 1>, rotation=<rot 0 86.6218 -87.0721>}]
 dyn_strap_fr_r_05 [{shapeType=Capsule, length=8.8881, radius=2.5, shapeName=dyn_strap_fr_l_05_capsule, center=<1.2871 0 1>, rotation=<rot 0 86.6218 -87.0721>}]
 dyn_strap_fr_r_06 [{shapeType=Capsule, length=6.7631, radius=2.5, shapeName=dyn_strap_fr_l_05_capsule, center=<1.2871 0 1>, rotation=<rot 0 86.6218 -87.0721>}]
 dyn_strap_fr_r_MSTR [{shapeType=Capsule, length=0.5, radius=0.5}]
 pelvis [
  {shapeType=Capsule, length=8.8463, radius=10.685, center=<-2.5033 -2.528 -0.0002>}
  {shapeType=Capsule, length=2.9455, radius=12.1716, center=<-6.1201 -1.0203 0.0003>, rotation=<rot 90 -0.0197 -0.0197>}
  {shapeType=Capsule, length=15, radius=6.1259, center=<4.538 -4.2399 0>}
 ]
 spine_01 [{shapeType=Capsule, length=6.125, radius=8.3759, center=<4.538 0.6326 -0.0001>}]
 spine_02 [{shapeType=Capsule, length=6.125, radius=8.3759, center=<4.538 0.6326 -0.0001>}]
 thigh_l [{shapeType=Capsule, length=32.475, radius=9.0562, center=<-21.1166 -0.8915 0.0108>, rotation=<rot 90 0 0>}]
 thigh_r [{shapeType=Capsule, length=32.475, radius=9.0562, center=<21.1166 0.9538 0.0528>, rotation=<rot 90 0 0>}]
}
constraints [
 {
  bone1Name=thigh_l
  bone2Name=pelvis
  swing1LimitDegrees=45
  swing1Motion=Free
  swing2LimitDegrees=45
  swing2Motion=Free
  twistLimitDegrees=45
  twistMotion=Free
 }
 {
  bone1Name=thigh_r
  bone2Name=pelvis
  swing1LimitDegrees=45
  swing1Motion=Free
  swing2LimitDegrees=45
  swing2Motion=Free
  twistLimitDegrees=45
  twistMotion=Free
 }
 {bone1Name=dyn_belt_r_01, bone2Name=pelvis, swing1LimitDegrees=1, swing2LimitDegrees=5, twistLimitDegrees=5, twistMotion=Locked}
 {bone1Name=dyn_belt_r_02, bone2Name=dyn_belt_r_01, swing1LimitDegrees=0.7, swing2LimitDegrees=10, twistLimitDegrees=22.28, twistMotion=Locked}
 {bone1Name=dyn_belt_r_03, bone2Name=dyn_belt_r_02, swing1LimitDegrees=0.7, swing2LimitDegrees=45, twistLimitDegrees=5, twistMotion=Locked}
 {bone1Name=dyn_belt_l_01, bone2Name=pelvis, swing1LimitDegrees=1, swing2LimitDegrees=5, twistLimitDegrees=5, twistMotion=Locked}
 {bone1Name=dyn_belt_l_02, bone2Name=dyn_belt_l_01, swing1LimitDegrees=5, swing2LimitDegrees=10, twistLimitDegrees=5, twistMotion=Locked}
 {
  bone1Name=dyn_belt_l_03
  bone2Name=dyn_belt_l_02
  swing1LimitDegrees=45
  swing1Motion=Locked
  swing2LimitDegrees=45
  twistLimitDegrees=5
  twistMotion=Locked
 }
 {bone1Name=dyn_strap_bk_mid_01, bone2Name=pelvis, swing1LimitDegrees=35, swing2LimitDegrees=35, twistLimitDegrees=5}
 {bone1Name=dyn_strap_bk_mid_02, bone2Name=dyn_strap_bk_mid_01, swing1LimitDegrees=35, swing2LimitDegrees=35, twistLimitDegrees=5}
 {bone1Name=dyn_strap_bk_mid_03, bone2Name=dyn_strap_bk_mid_02, swing1LimitDegrees=35, swing2LimitDegrees=35, twistLimitDegrees=5}
 {bone1Name=dyn_strap_bk_l_01, bone2Name=dyn_strap_bk_mid_03, swing1LimitDegrees=45, swing2LimitDegrees=45, twistLimitDegrees=5}
 {bone1Name=dyn_strap_bk_l_02, bone2Name=dyn_strap_bk_l_01, swing1LimitDegrees=10, swing2LimitDegrees=45, twistLimitDegrees=5}
 {bone1Name=dyn_strap_bk_l_03, bone2Name=dyn_strap_bk_l_02, swing1LimitDegrees=10, swing2LimitDegrees=45, twistLimitDegrees=5}
 {bone1Name=dyn_strap_bk_l_04, bone2Name=dyn_strap_bk_l_03, swing1LimitDegrees=10, swing2LimitDegrees=45, twistLimitDegrees=5}
 {bone1Name=dyn_strap_bk_l_05, bone2Name=dyn_strap_bk_l_04, swing1LimitDegrees=10, swing2LimitDegrees=45, twistLimitDegrees=5}
 {bone1Name=dyn_strap_bk_r_01, bone2Name=dyn_strap_bk_mid_03, swing1LimitDegrees=45, swing2LimitDegrees=45, twistLimitDegrees=5}
 {bone1Name=dyn_strap_bk_r_02, bone2Name=dyn_strap_bk_r_01, swing1LimitDegrees=45, swing2LimitDegrees=45, twistLimitDegrees=5}
 {bone1Name=dyn_strap_bk_r_03, bone2Name=dyn_strap_bk_r_02, swing1LimitDegrees=45, swing2LimitDegrees=45, twistLimitDegrees=5}
 {bone1Name=dyn_strap_bk_r_04, bone2Name=dyn_strap_bk_r_03, swing1LimitDegrees=45, swing2LimitDegrees=45, twistLimitDegrees=5}
 {bone1Name=dyn_strap_bk_r_05, bone2Name=dyn_strap_bk_r_04, swing1LimitDegrees=45, swing2LimitDegrees=45, twistLimitDegrees=5}
 {
  bone1Name=dyn_belt_mid_MSTR
  bone2Name=pelvis
  swing1LimitDegrees=45
  swing1Motion=Free
  swing2LimitDegrees=45
  swing2Motion=Free
  twistLimitDegrees=45
  twistMotion=Free
 }
 {bone1Name=dyn_belt_mid_01, bone2Name=dyn_belt_mid_MSTR, swing1LimitDegrees=45, swing2LimitDegrees=45, twistLimitDegrees=5}
 {bone1Name=dyn_belt_mid_02, bone2Name=dyn_belt_mid_01, swing1LimitDegrees=45, swing2LimitDegrees=45, twistLimitDegrees=5}
 {
  bone1Name=dyn_strap_fr_r_MSTR
  bone2Name=pelvis
  swing1LimitDegrees=45
  swing1Motion=Free
  swing2LimitDegrees=45
  swing2Motion=Free
  twistLimitDegrees=45
  twistMotion=Free
 }
 {bone1Name=dyn_strap_fr_r_01, bone2Name=dyn_strap_fr_r_MSTR, swing1LimitDegrees=5, swing2LimitDegrees=60, twistLimitDegrees=5}
 {bone1Name=dyn_strap_fr_r_02, bone2Name=dyn_strap_fr_r_01, swing1LimitDegrees=35, swing2LimitDegrees=45, twistLimitDegrees=0}
 {bone1Name=dyn_strap_fr_r_03, bone2Name=dyn_strap_fr_r_02a, swing1LimitDegrees=45, swing2LimitDegrees=45, twistLimitDegrees=5}
 {bone1Name=dyn_strap_fr_r_04, bone2Name=dyn_strap_fr_r_03, swing1LimitDegrees=45, swing2LimitDegrees=45, twistLimitDegrees=5}
 {bone1Name=dyn_strap_fr_r_05, bone2Name=dyn_strap_fr_r_04, swing1LimitDegrees=45, swing2LimitDegrees=45, twistLimitDegrees=5}
 {
  bone1Name=dyn_strap_fr_l_MSTR
  bone2Name=pelvis
  swing1LimitDegrees=45
  swing1Motion=Free
  swing2LimitDegrees=45
  swing2Motion=Free
  twistLimitDegrees=45
  twistMotion=Free
 }
 {bone1Name=dyn_strap_fr_l_01, bone2Name=dyn_strap_fr_l_MSTR, swing1LimitDegrees=5, swing2LimitDegrees=60, twistLimitDegrees=5}
 {bone1Name=dyn_strap_fr_l_02, bone2Name=dyn_strap_fr_l_01, swing1LimitDegrees=35, swing2LimitDegrees=45, twistLimitDegrees=5}
 {bone1Name=dyn_strap_fr_l_03, bone2Name=dyn_strap_fr_l_02a, swing1LimitDegrees=45, swing2LimitDegrees=45, twistLimitDegrees=5}
 {bone1Name=dyn_strap_fr_l_04, bone2Name=dyn_strap_fr_l_03, swing1LimitDegrees=45, swing2LimitDegrees=45, twistLimitDegrees=5}
 {bone1Name=dyn_strap_fr_l_05, bone2Name=dyn_strap_fr_l_04, swing1LimitDegrees=45, swing2LimitDegrees=45, twistLimitDegrees=5}
 {
  bone1Name=dyn_belt_l_03
  bone2Name=dyn_belt_mid_02
  swing1LimitDegrees=45
  swing1Motion=Free
  swing2LimitDegrees=45
  swing2Motion=Free
  twistLimitDegrees=45
  twistMotion=Free
 }
 {
  bone1Name=dyn_belt_r_03
  bone2Name=dyn_belt_mid_02
  swing1LimitDegrees=45
  swing1Motion=Free
  swing2LimitDegrees=45
  swing2Motion=Free
  twistLimitDegrees=45
  twistMotion=Free
 }
 {bone1Name=dyn_strap_fr_r_02a, bone2Name=dyn_strap_fr_r_02, swing1LimitDegrees=25, swing2LimitDegrees=25, twistLimitDegrees=5}
 {bone1Name=dyn_strap_fr_l_02a, bone2Name=dyn_strap_fr_l_02, swing1LimitDegrees=25, swing2LimitDegrees=25, twistLimitDegrees=5}
 {bone1Name=dyn_strap_bk_r_06, bone2Name=dyn_strap_bk_r_05, swing1LimitDegrees=45, swing2LimitDegrees=45, twistLimitDegrees=5}
 {bone1Name=dyn_strap_fr_r_06, bone2Name=dyn_strap_fr_r_05, swing1LimitDegrees=45, swing2LimitDegrees=45, twistLimitDegrees=5}
 {bone1Name=dyn_strap_bk_l_06, bone2Name=dyn_strap_bk_l_05, swing1LimitDegrees=45, swing2LimitDegrees=45, twistLimitDegrees=5}
 {bone1Name=dyn_strap_fr_l_06, bone2Name=dyn_strap_fr_l_05, swing1LimitDegrees=45, swing2LimitDegrees=45, twistLimitDegrees=5}
 {
  bone1Name=dyn_strap_fr_l_06
  bone2Name=dyn_strap_bk_l_06
  swing1LimitDegrees=45
  swing1Motion=Free
  swing2LimitDegrees=45
  swing2Motion=Free
  twistLimitDegrees=45
  twistMotion=Free
 }
 {
  bone1Name=dyn_strap_fr_r_06
  bone2Name=dyn_strap_bk_r_06
  swing1LimitDegrees=45
  swing1Motion=Free
  swing2LimitDegrees=45
  swing2Motion=Free
  twistLimitDegrees=45
  twistMotion=Free
 }
 {
  bone1Name=spine_01
  bone2Name=pelvis
  swing1LimitDegrees=45
  swing1Motion=Free
  swing2LimitDegrees=45
  swing2Motion=Free
  twistLimitDegrees=45
  twistMotion=Free
 }
 {
  bone1Name=spine_02
  bone2Name=spine_01
  swing1LimitDegrees=45
  swing1Motion=Free
  swing2LimitDegrees=45
  swing2Motion=Free
  twistLimitDegrees=45
  twistMotion=Free
 }
 {
  bone1Name=thigh_l
  bone2Name=dyn_strap_fr_l_01
  swing1LimitDegrees=45
  swing1Motion=Free
  swing2LimitDegrees=45
  swing2Motion=Free
  twistLimitDegrees=45
  twistMotion=Free
 }
 {
  bone1Name=thigh_r
  bone2Name=dyn_strap_fr_r_01
  swing1LimitDegrees=45
  swing1Motion=Free
  swing2LimitDegrees=45
  swing2Motion=Free
  twistLimitDegrees=45
  twistMotion=Free
 }
 {
  bone1Name=dyn_strap_bk_l_06
  bone2Name=dyn_strap_fr_l_05
  swing1LimitDegrees=45
  swing1Motion=Free
  swing2LimitDegrees=45
  swing2Motion=Free
  twistLimitDegrees=45
  twistMotion=Free
 }
]
