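# Physics asset for the dynamic necklace charms on the female medium Lollipop Trickster player body.
# Five charms each pair a small MSTR anchor capsule with a shaped charm capsule, and one large spine_05 capsule represents the upper torso.
# Anchors join spine_05 through angularly free constraints, while charms have angular constraints to their anchors.
# The A charms lock twist and one swing axis to act as hinges; B and C keep all three axes unlocked.
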PhysicsAsset /Game/Characters/Player/Female/Medium/Bodies/F_MED_Lollipop_Trickster/Meshes/F_MED_Lollipop_Trickster_Dyn_charms
bodies {
 dyn_necklace_charm_A_l_01 [{shapeType=Capsule, length=1.6258, radius=1.4395, center=<1.5693 0 -0.0001>, rotation=<rot 0 0 90>}]
 dyn_necklace_charm_A_l_MSTR [{shapeType=Capsule, length=0.5, radius=0.5}]
 dyn_necklace_charm_A_r_01 [{shapeType=Capsule, length=0.4417, radius=2.3091, center=<1.2 0 0>, rotation=<rot 90 0 0>}]
 dyn_necklace_charm_A_r_MSTR [{shapeType=Capsule, length=0.5, radius=0.5}]
 dyn_necklace_charm_B_l_01 [{shapeType=Capsule, length=2.4545, radius=0.7161, center=<2.1365 0 0>, rotation=<rot 90 0 0>}]
 dyn_necklace_charm_B_l_MSTR [{shapeType=Capsule, length=0.5, radius=0.5}]
 dyn_necklace_charm_B_r_01 [{shapeType=Capsule, length=1.5904, radius=0.8856, center=<2.4148 -0.093 0>, rotation=<rot 90 0 0>}]
 dyn_necklace_charm_B_r_MSTR [{shapeType=Capsule, length=0.5, radius=0.5}]
 dyn_necklace_charm_C_l_01 [{shapeType=Capsule, length=2.2354, radius=0.7704, center=<1.6801 0.1479 0>, rotation=<rot 90 0 0>}]
 dyn_necklace_charm_C_l_MSTR [{shapeType=Capsule, length=0.5, radius=0.5}]
 spine_05 [{shapeType=Capsule, length=20.3375, radius=8.1032, shapeName=spine_05_capsule, center=<1.5443 -1.6046 0>}]
}
constraints [
 {
  bone1Name=dyn_necklace_charm_A_l_01
  bone2Name=dyn_necklace_charm_A_l_MSTR
  swing1LimitDegrees=5
  swing1Motion=Locked
  swing2LimitDegrees=40
  twistLimitDegrees=5
  twistMotion=Locked
 }
 {
  bone1Name=dyn_necklace_charm_B_l_01
  bone2Name=dyn_necklace_charm_B_l_MSTR
  swing1LimitDegrees=40
  swing2LimitDegrees=40
  twistLimitDegrees=15
 }
 {
  bone1Name=dyn_necklace_charm_C_l_01
  bone2Name=dyn_necklace_charm_C_l_MSTR
  swing1LimitDegrees=40
  swing2LimitDegrees=40
  twistLimitDegrees=5
 }
 {
  bone1Name=dyn_necklace_charm_B_r_01
  bone2Name=dyn_necklace_charm_B_r_MSTR
  swing1LimitDegrees=40
  swing2LimitDegrees=40
  twistLimitDegrees=5
 }
 {
  bone1Name=dyn_necklace_charm_A_r_01
  bone2Name=dyn_necklace_charm_A_r_MSTR
  swing1LimitDegrees=5
  swing1Motion=Locked
  swing2LimitDegrees=40
  twistLimitDegrees=5
  twistMotion=Locked
 }
 {
  bone1Name=dyn_necklace_charm_A_l_MSTR
  bone2Name=spine_05
  swing1LimitDegrees=45
  swing1Motion=Free
  swing2LimitDegrees=45
  swing2Motion=Free
  twistLimitDegrees=45
  twistMotion=Free
 }
 {
  bone1Name=dyn_necklace_charm_B_l_MSTR
  bone2Name=spine_05
  swing1LimitDegrees=45
  swing1Motion=Free
  swing2LimitDegrees=45
  swing2Motion=Free
  twistLimitDegrees=45
  twistMotion=Free
 }
 {
  bone1Name=dyn_necklace_charm_C_l_MSTR
  bone2Name=spine_05
  swing1LimitDegrees=45
  swing1Motion=Free
  swing2LimitDegrees=45
  swing2Motion=Free
  twistLimitDegrees=45
  twistMotion=Free
 }
 {
  bone1Name=dyn_necklace_charm_B_r_MSTR
  bone2Name=spine_05
  swing1LimitDegrees=45
  swing1Motion=Free
  swing2LimitDegrees=45
  swing2Motion=Free
  twistLimitDegrees=45
  twistMotion=Free
 }
 {
  bone1Name=dyn_necklace_charm_A_r_MSTR
  bone2Name=spine_05
  swing1LimitDegrees=45
  swing1Motion=Free
  swing2LimitDegrees=45
  swing2Motion=Free
  twistLimitDegrees=45
  twistMotion=Free
 }
]
