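# Physics asset for the left leg of the F_MED_Lollipop_Trickster female player body, covering two dynamic boot bones.
# A small capsule body on calf_twist_01_l is the anchor, and dyn_boot_l_01 and dyn_boot_l_02 each carry an identical, larger capsule.
# Each boot bone is constrained directly to the calf twist bone rather than chained, with moderate swing limits and a tight twist limit.
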PhysicsAsset /Game/Characters/Player/Female/Medium/Bodies/F_MED_Lollipop_Trickster/Meshes/F_MED_Lollipop_Trickster_Dyn_leg_l
bodies {
 calf_twist_01_l [{shapeType=Capsule, length=0.5, radius=0.5}]
 dyn_boot_l_01 [{shapeType=Capsule, length=3, radius=1.1562, center=<2.5062 -0.0001 0>, rotation=<rot 90 0 0>}]
 dyn_boot_l_02 [{shapeType=Capsule, length=3, radius=1.1562, center=<2.5062 -0.0001 0>, rotation=<rot 90 0 0>}]
}
constraints [
 {bone1Name=dyn_boot_l_02, bone2Name=calf_twist_01_l, swing1LimitDegrees=45, swing2LimitDegrees=45, twistLimitDegrees=5}
 {bone1Name=dyn_boot_l_01, bone2Name=calf_twist_01_l, swing1LimitDegrees=45, swing2LimitDegrees=45, twistLimitDegrees=5}
]
