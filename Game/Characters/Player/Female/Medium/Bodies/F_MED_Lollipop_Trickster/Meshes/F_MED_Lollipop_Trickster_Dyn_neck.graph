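# Physics asset for the dynamic necklace on the Lollipop Trickster medium female player body.
# A necklace base, angularly locked to neck_01, roots two jointed chains: a three-capsule main strand and a four-link back strand of small, near-spherical capsules, with narrow twist limits.
# Two large mirrored capsules on spine_05 approximate the upper torso, and neck_01 joins that bone with free swing and twist.
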PhysicsAsset /Game/Characters/Player/Female/Medium/Bodies/F_MED_Lollipop_Trickster/Meshes/F_MED_Lollipop_Trickster_Dyn_neck
bodies {
 dyn_necklace_base [{shapeType=Sphere, radius=1.1223}]
 dyn_necklace_bk_01 [{shapeType=Capsule, length=0.0004, radius=1.2454, center=<1 0 0>, rotation=<rot 90 0 0>}]
 dyn_necklace_bk_02 [{shapeType=Capsule, length=0, radius=1.0528, center=<1 0 0>, rotation=<rot 90 0 0>}]
 dyn_necklace_bk_03 [{shapeType=Capsule, length=0.0197, radius=0.9903, center=<1 0 0>, rotation=<rot 90 0 0>}]
 dyn_necklace_bk_04 [{shapeType=Capsule, length=0.0197, radius=0.9903, center=<1 0 0>, rotation=<rot 90 0 0>}]
 dyn_necklace_main_01 [{
   shapeType=Capsule
   length=6.4637
   radius=1.8871
   shapeName=dyn_necklace_main_01_capsule
   center=<2.465 -0.1503 0.1935>
   rotation=<rot 0 -179.6444 -89.633>
  }]
 dyn_necklace_main_02 [{
   shapeType=Capsule
   length=1.7149
   radius=1.7787
   shapeName=dyn_necklace_main_02_capsule
   center=<3.0214 0.0445 0.0154>
   rotation=<rot 0 179.1516 -88.1484>
  }]
 dyn_necklace_main_03 [{shapeType=Capsule, length=1.637, radius=1.3626, shapeName=dyn_necklace_main_03_capsule, center=<2.5732 0.161 -0.0841>, rotation=<rot 0 0 90>}]
 neck_01 [{shapeType=Capsule, length=0.505, radius=0.505}]
 spine_05 [
  {shapeType=Capsule, length=15.1313, radius=3.9738, shapeName=spine_05_capsule, center=<2.668 2.9506 2.3335>, rotation=<rot 0 47.8776 89.9999>}
  {shapeType=Capsule, length=15.1313, radius=3.9738, shapeName=spine_05_capsule, center=<2.5997 2.761 -2.949>, rotation=<rot 0 47.8757 89.9999>}
 ]
}
constraints [
 {bone1Name=dyn_necklace_bk_01, bone2Name=dyn_necklace_base, swing1LimitDegrees=45, swing2LimitDegrees=45, twistLimitDegrees=5}
 {bone1Name=dyn_necklace_bk_02, bone2Name=dyn_necklace_bk_01, swing1LimitDegrees=45, swing2LimitDegrees=45, twistLimitDegrees=5}
 {bone1Name=dyn_necklace_bk_03, bone2Name=dyn_necklace_bk_02, swing1LimitDegrees=45, swing2LimitDegrees=45, twistLimitDegrees=5}
 {bone1Name=dyn_necklace_bk_04, bone2Name=dyn_necklace_bk_03, swing1LimitDegrees=45, swing2LimitDegrees=45, twistLimitDegrees=5}
 {
  bone1Name=dyn_necklace_base
  bone2Name=neck_01
  swing1LimitDegrees=45
  swing1Motion=Locked
  swing2LimitDegrees=25
  swing2Motion=Locked
  twistLimitDegrees=5
  twistMotion=Locked
 }
 {
  bone1Name=neck_01
  bone2Name=spine_05
  swing1LimitDegrees=45
  swing1Motion=Free
  swing2LimitDegrees=45
  swing2Motion=Free
  twistLimitDegrees=45
  twistMotion=Free
 }
 {bone1Name=dyn_necklace_main_01, bone2Name=dyn_necklace_base, swing1LimitDegrees=45, swing2LimitDegrees=25, twistLimitDegrees=5}
 {bone1Name=dyn_necklace_main_02, bone2Name=dyn_necklace_main_01, swing1LimitDegrees=45, swing2LimitDegrees=45, twistLimitDegrees=5}
 {bone1Name=dyn_necklace_main_03, bone2Name=dyn_necklace_main_02, swing1LimitDegrees=45, swing2LimitDegrees=45, twistLimitDegrees=5}
]
