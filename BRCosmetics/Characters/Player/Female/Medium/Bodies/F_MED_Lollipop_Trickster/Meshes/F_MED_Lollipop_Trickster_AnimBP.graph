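# Animation Blueprint for the F_MED_Lollipop_Trickster body cosmetic, adding secondary motion to a pose copied from the attached parent mesh.
# It chains pelvis and strap correctives, pelvis RigidBody, Anim Dynamics necklace and belt chains, then constrained, head-corrected charms.
# The event graph derives corrective alphas from thigh and neck deltas, and varies RigidBody gravity and blend for movement states like skydiving.
# Four other RigidBody nodes never reach the output.
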AnimBlueprint /BRCosmetics/Characters/Player/Female/Medium/Bodies/F_MED_Lollipop_Trickster/Meshes/F_MED_Lollipop_Trickster_AnimBP
target_skeleton=&/F_MED_Lollipop_Trickster_Skeleton
graphs [
 AnimGraph {nodes [
   RigidBody() {
    bClampLinearTranslationLimitToRefPose=true
    bOverrideWorldGravity=true
    overridePhysicsAsset=/Game/Characters/Player/Female/Medium/Bodies/F_MED_Lollipop_Trickster/Meshes/F_MED_Lollipop_Trickster_Dyn_charms
    simulationSpace=BaseBoneSpace
    alphaBoolBlend {blendOption=Linear}
    baseBoneRef {boneName=spine_05}
    componentAppliedLinearAccClamp=<3000 3000 3000>
    componentLinearAccScale=<0.3 0.2 0.2>
    componentLinearVelScale=<0.3 0.2 0.2>
    overrideWorldGravity {z=-980}
   }
   n1 = RigidBody() {
    bClampLinearTranslationLimitToRefPose=true
    bOverrideWorldGravity=true
    overridePhysicsAsset=/Game/Characters/Player/Female/Medium/Bodies/F_MED_Lollipop_Trickster/Meshes/F_MED_Lollipop_Trickster_Dyn_leg_l
    simulationSpace=BaseBoneSpace
    alphaBoolBlend {blendOption=Linear}
    baseBoneRef {boneName=calf_l}
    componentAppliedLinearAccClamp=<3000 3000 3000>
    componentLinearAccScale=<0.5 0.4 0.4>
    componentLinearVelScale=<0.5 0.4 0.4>
    overrideWorldGravity {z=-980}
   }
   RigidBody(ComponentPose=@n1) {
    bClampLinearTranslationLimitToRefPose=true
    bOverrideWorldGravity=true
    overridePhysicsAsset=/Game/Characters/Player/Female/Medium/Bodies/F_MED_Lollipop_Trickster/Meshes/F_MED_Lollipop_Trickster_Dyn_leg_r
    simulationSpace=BaseBoneSpace
    alphaBoolBlend {blendOption=Linear}
    baseBoneRef {boneName=calf_r}
    componentAppliedLinearAccClamp=<3000 3000 3000>
    componentLinearAccScale=<0.5 0.4 0.4>
    componentLinearVelScale=<0.5 0.4 0.4>
    overrideWorldGravity {z=-980}
   }
   RigidBody!(ComponentPose, Alpha, OverrideWorldGravity:Vector="0,0,0", ExternalForce) -> (Pose) {
    bClampLinearTranslationLimitToRefPose=true
    bOverrideWorldGravity=true
    overridePhysicsAsset=/Game/Characters/Player/Female/Medium/Bodies/F_MED_Lollipop_Trickster/Meshes/F_MED_Lollipop_Trickster_Dyn_neck
    simulationSpace=BaseBoneSpace
    alphaBoolBlend {blendOption=Linear}
    baseBoneRef {boneName=spine_04}
    componentAppliedLinearAccClamp {y=3000, z=3000}
    componentLinearAccScale=<0.4 0.2 0.2>
    componentLinearVelScale=<0.4 0.2 0.2>
    overrideWorldGravity {z=-980}
   }
   n2 = Get thigh_rot_l()
   n3 = "Transform (Modify) Bone - Bone: pelvis_l"(
    ComponentPose=Local To Component(LocalPose=Copy Pose From Mesh() {bUseAttachedParent=true})
    Alpha=@n2
    Translation="-6.240,-1.520,1.245"
    Rotation="0,0,0"
    Scale="1,1,1"
   ) {
    scaleMode=BMM_Additive
    translationMode=BMM_Additive
    translationSpace=BCS_ParentBoneSpace
    alphaBoolBlend {blendOption=Linear}
    boneToModify {boneName=pelvis_l}
    translation=<-6.24 -1.52 1.245>
   }
   n4 = "Transform (Modify) Bone - Bone: pelvis_mid_l"(ComponentPose=@n3, Alpha=@n2, Translation="-1.450,-1.138,0.518", Rotation="0,0,0", Scale="1,1,1") {
    scaleMode=BMM_Additive
    translationMode=BMM_Additive
    translationSpace=BCS_ParentBoneSpace
    alphaBoolBlend {blendOption=Linear}
    boneToModify {boneName=pelvis_mid_l}
    translation=<-1.45 -1.138 0.518>
   }
   n5 = "Transform (Modify) Bone - Bone: pelvis_side_l"(ComponentPose=@n4, Alpha=@n2, Translation="0.505,-0.560,-1.440", Rotation="0,0,0", Scale="1,1,1") {
    scaleMode=BMM_Additive
    translationMode=BMM_Additive
    translationSpace=BCS_ParentBoneSpace
    alphaBoolBlend {blendOption=Linear}
    boneToModify {boneName=pelvis_side_l}
    translation=<0.505 -0.56 -1.44>
   }
   n6 = Get thigh_rot_r()
   n7 = "Transform (Modify) Bone - Bone: pelvis_r"(ComponentPose=@n5, Alpha=@n6, Translation="-1.106,-0.155,0.183", Rotation="0,0,0", Scale="1,1,1") {
    scaleMode=BMM_Additive
    translationMode=BMM_Additive
    translationSpace=BCS_ParentBoneSpace
    alphaBoolBlend {blendOption=Linear}
    boneToModify {boneName=pelvis_r}
    translation=<-1.106 -0.155 0.183>
   }
   n8 = "Transform (Modify) Bone - Bone: pelvis_mid_r"(ComponentPose=@n7, Alpha=@n6, Translation="-0.848,0.002,0.106", Rotation="0,0,0", Scale="1,1,1") {
    scaleMode=BMM_Additive
    translationMode=BMM_Additive
    translationSpace=BCS_ParentBoneSpace
    alphaBoolBlend {blendOption=Linear}
    boneToModify {boneName=pelvis_mid_r}
    translation=<-0.848 0.002 0.106>
   }
   n9 = "Transform (Modify) Bone - Bone: pelvis_side_r"(ComponentPose=@n8, Alpha=@n6, Translation="-0.144,-0.340,-1.135", Rotation="0,0,0", Scale="1,1,1") {
    scaleMode=BMM_Additive
    translationMode=BMM_Additive
    translationSpace=BCS_ParentBoneSpace
    alphaBoolBlend {blendOption=Linear}
    boneToModify {boneName=pelvis_side_r}
    translation=<-0.144 -0.34 -1.135>
   }
   n10 = "Transform (Modify) Bone - Bone: dyn_strap_fr_r_MSTR"!(ComponentPose=@n9, bAlphaBoolEnabled:bool=Get bIsInFrontEnd(), Translation="0,0,0", Rotation="0,45,0", Scale="1,1,1") -> (Pose) {
    alphaInputType=Bool
    rotationMode=BMM_Additive
    rotationSpace=BCS_BoneSpace
    scaleSpace=BCS_BoneSpace
    translationMode=BMM_Additive
    translationSpace=BCS_BoneSpace
    alphaBoolBlend {blendOption=Linear}
    alphaScaleBiasClamp {bInterpResult=true, interpSpeedDecreasing=15, interpSpeedIncreasing=15}
    boneToModify {boneName=dyn_strap_fr_r_MSTR}
    rotation {yaw=45}
   }
   n11 = "Transform (Modify) Bone - Bone: dyn_strap_fr_l_MSTR"(ComponentPose=@n10, Alpha=Get thigh_rot_l(), Translation="0.000,-0.517,2.019", Rotation="0,0.000,-20", Scale="1,1,1") {
    rotationMode=BMM_Additive
    translationMode=BMM_Additive
    alphaBoolBlend {blendOption=Linear}
    alphaScaleBiasClamp {bInterpResult=true, interpSpeedDecreasing=15, interpSpeedIncreasing=15}
    boneToModify {boneName=dyn_strap_fr_l_MSTR}
    rotation {roll=-20}
    translation {y=-0.517, z=2.019}
   }
   n12 = "Transform (Modify) Bone - Bone: dyn_strap_fr_r_MSTR"!(ComponentPose=@n11, Alpha:float=Get thigh_rot_r(), Translation="0.674,-0.985,3.016", Rotation="0.000,0.000,20", Scale="1,1,1") -> (Pose) {
    rotationMode=BMM_Additive
    translationMode=BMM_Additive
    alphaBoolBlend {blendOption=Linear}
    alphaScaleBiasClamp {bInterpResult=true, interpSpeedDecreasing=15, interpSpeedIncreasing=15}
    boneToModify {boneName=dyn_strap_fr_r_MSTR}
    rotation {roll=20}
    translation=<0.674 -0.985 3.016>
   }
   n13 = "Transform (Modify) Bone - Bone: dyn_strap_fr_r_MSTR"(ComponentPose=@n12, Alpha=Get Thigh_R_Rot_Crouch() -> (Thigh_R_Rot_Crouch:double), Translation="0,0,0", Rotation="0,90,45", Scale="1,1,1") {
    rotationMode=BMM_Additive
    rotationSpace=BCS_BoneSpace
    alphaBoolBlend {blendOption=Linear}
    boneToModify {boneName=dyn_strap_fr_r_MSTR}
    rotation {roll=45, yaw=90}
   }
   n14 = "Transform (Modify) Bone - Bone: dyn_strap_fr_l_MSTR"(ComponentPose=@n13, Alpha=Get Thigh L Rot Crouch() -> (Thigh L Rot Crouch:double), Translation="0,0,0", Rotation="0,-90,-45", Scale="1,1,1") {
    rotationMode=BMM_Additive
    rotationSpace=BCS_BoneSpace
    alphaBoolBlend {blendOption=Linear}
    boneToModify {boneName=dyn_strap_fr_l_MSTR}
    rotation {roll=-45, yaw=-90}
   }
   n15 = RigidBody(
    ComponentPose=@n14
    Alpha=Get Skydive Alpha() -> (Skydive Alpha:double)
    OverrideWorldGravity=Get gravity_override() -> (gravity_override:Vector)
   ) {
    bClampLinearTranslationLimitToRefPose=true
    bOverrideWorldGravity=true
    overridePhysicsAsset=&/F_MED_Lollipop_Trickster_Dyn
    simulationSpace=BaseBoneSpace
    alphaBoolBlend {blendOption=Linear}
    baseBoneRef {boneName=pelvis}
    componentAppliedLinearAccClamp=<3000 3000 3000>
    componentLinearAccScale=<0.2 0.2 0.2>
    componentLinearVelScale=<0.2 0.1 0.1>
    overrideWorldGravity {z=-980}
   }
   n16 = "Transform (Modify) Bone - Bone: dyn_necklace_bk_01"(
    ComponentPose=@n15
    Alpha=Get Head Back Rotation Alpha() -> (Head Back Rotation Alpha:double)
    Translation="0,0,1.25"
    Rotation="45,0,0"
    Scale="1,1,1"
   ) {
    rotationMode=BMM_Additive
    rotationSpace=BCS_BoneSpace
    translationMode=BMM_Additive
    translationSpace=BCS_BoneSpace
    alphaBoolBlend {blendOption=Linear}
    boneToModify {boneName=dyn_necklace_bk_01}
    rotation {pitch=45}
    translation {z=1.25}
   }
   n17 = "Anim Dynamics - Chain: dyn_necklace_bk_02 -> dyn_necklace_bk_04"(ComponentPose=@n16, Alpha=1, ExternalForce="0,0,0") {
    angularDampingOverride=0.9993
    bChain=true
    bGravityOverrideInSimSpace=true
    bOverrideAngularDamping=true
    bUseGravityOverride=true
    bUsePlanarLimit=false
    bUseSphericalLimits=true
    alphaBoolBlend {blendOption=Linear}
    boundBone {boneName=dyn_necklace_bk_02}
    chainEnd {boneName=dyn_necklace_bk_04}
    componentAppliedLinearAccClamp=<3000 3000 3000>
    componentLinearAccScale=<0.3 0.3 0.333>
    componentLinearVelScale=<0.3 0.3 0.3>
    gravityOverride {z=-860}
    physicsBodyDefinitions [{
      sphereCollisionRadius=0
      boundBone {boneName=dyn_necklace_bk_02}
      boxExtents=<20 20 10>
      constraintSetup {angularLimitsMax=<2 10 20>, angularLimitsMin=<-2 -60 -20>, angularTarget=<0 0 0>}
      localJointOffset=<3 0 0>
     }]
    sphericalLimits [
     {limitRadius=12, drivingBone {boneName=spine_05}, sphereLocalOffset=<2 3.25 0>}
     {limitRadius=12, drivingBone {boneName=spine_05}, sphereLocalOffset=<2 3.25 4>}
     {limitRadius=12, drivingBone {boneName=spine_05}, sphereLocalOffset=<2 3.25 -4>}
    ]
   }
   n18 = "Anim Dynamics - Chain: dyn_necklace_main_01 -> dyn_necklace_main_03"(ComponentPose=@n17, Alpha=1, ExternalForce="0,0,0") {
    angularDampingOverride=0.9995
    bChain=true
    bGravityOverrideInSimSpace=true
    bOverrideAngularDamping=true
    bUseGravityOverride=true
    bUsePlanarLimit=false
    bUseSphericalLimits=true
    alphaBoolBlend {blendOption=Linear}
    boundBone {boneName=dyn_necklace_main_01}
    chainEnd {boneName=dyn_necklace_main_03}
    componentAppliedLinearAccClamp=<5000 5000 5000>
    componentLinearAccScale=<0.3 0.3 0.3>
    componentLinearVelScale=<0.3 0.3 0.3>
    gravityOverride {z=-860}
    physicsBodyDefinitions [{
      sphereCollisionRadius=0
      boundBone {boneName=dyn_necklace_main_01}
      boxExtents=<35 35 8>
      constraintSetup {angularLimitsMax=<2 15 10>, angularLimitsMin=<-2 -60 -10>, angularTarget=<0 0 0>}
      localJointOffset=<3 0 0>
     }]
    sphericalLimits [
     {limitRadius=12, drivingBone {boneName=spine_05}, sphereLocalOffset=<0 -3.25 0>}
     {limitRadius=12, drivingBone {boneName=spine_05}, sphereLocalOffset=<0 -3.25 4>}
     {limitRadius=12, drivingBone {boneName=spine_05}, sphereLocalOffset=<0 -3.25 -4>}
    ]
   }
   n19 = "Anim Dynamics - Chain: dyn_belt_mid_01 -> dyn_belt_mid_02"(ComponentPose=@n18, Alpha=1, ExternalForce="0,0,0") {
    angularDampingOverride=0.9993
    bChain=true
    bGravityOverrideInSimSpace=true
    bOverrideAngularDamping=true
    bUseGravityOverride=true
    bUsePlanarLimit=false
    bUseSphericalLimits=true
    alphaBoolBlend {blendOption=Linear}
    boundBone {boneName=dyn_belt_mid_01}
    chainEnd {boneName=dyn_belt_mid_02}
    componentAppliedLinearAccClamp=<3000 3000 3000>
    componentLinearAccScale=<0.3 0.3 0.3>
    componentLinearVelScale=<0.3 0.3 0.3>
    gravityOverride {z=-860}
    physicsBodyDefinitions [
     {
      sphereCollisionRadius=0
      boundBone {boneName=dyn_belt_mid_01}
      boxExtents=<12 12 4>
      constraintSetup {angularLimitsMax=<2 6 15>, angularLimitsMin=<-2 -60 -15>, angularTarget=<0 0 0>}
      localJointOffset=<3 0 0>
     }
     {
      boundBone {boneName=dyn_belt_mid_02}
      boxExtents=<10 10 10>
      constraintSetup {angularLimitsMax=<0 0 0>, angularLimitsMin=<0 0 0>, angularTarget=<0 0 0>}
      localJointOffset=<0 0 0>
     }
    ]
    sphericalLimits [
     {limitRadius=12, drivingBone {boneName=spine_05}, sphereLocalOffset=<0 -3.25 0>}
     {limitRadius=12, drivingBone {boneName=spine_05}, sphereLocalOffset=<0 -3.25 4>}
     {limitRadius=12, drivingBone {boneName=spine_05}, sphereLocalOffset=<0 -3.25 -4>}
    ]
   }
   n20 = "Constraint - dyn_necklace_charm_A_l_MSTR (PP)"!(ComponentPose=@n19, Alpha=1, ConstraintWeights_0=0.7146, ConstraintWeights_1:float=0.2854) -> (Pose) {
    alphaBoolBlend {blendOption=Linear}
    boneToModify {boneName=dyn_necklace_charm_A_l_MSTR}
    constraintSetup [{targetBone {boneName=dyn_necklace_base}}, {targetBone {boneName=dyn_necklace_main_01}}]
    constraintWeights [1,1]
   }
   n21 = "Constraint - dyn_necklace_charm_B_l_MSTR (PPP)"!(ComponentPose=@n20, Alpha=1, ConstraintWeights_0=0.0283, ConstraintWeights_1:float=0.4338, ConstraintWeights_2:float=0.5379) -> (Pose) {
    alphaBoolBlend {blendOption=Linear}
    boneToModify {boneName=dyn_necklace_charm_B_l_MSTR}
    constraintSetup [{targetBone {boneName=dyn_necklace_main_02}}, {targetBone {boneName=dyn_necklace_main_01}}, {targetBone {boneName=dyn_necklace_base}}]
    constraintWeights [1,1,1]
   }
   n22 = "Constraint - dyn_necklace_charm_C_l_MSTR (PPP)"!(ComponentPose=@n21, Alpha=1, ConstraintWeights_0=0.1508, ConstraintWeights_1:float=0.7196, ConstraintWeights_2:float=0.1295) -> (Pose) {
    alphaBoolBlend {blendOption=Linear}
    boneToModify {boneName=dyn_necklace_charm_C_l_MSTR}
    constraintSetup [{targetBone {boneName=dyn_necklace_main_02}}, {targetBone {boneName=dyn_necklace_main_01}}, {targetBone {boneName=dyn_necklace_base}}]
    constraintWeights [1,1,1]
   }
   n23 = "Constraint - dyn_necklace_charm_B_r_MSTR (PPP)"!(ComponentPose=@n22, Alpha=1, ConstraintWeights_0=0.183, ConstraintWeights_1:float=0.7754, ConstraintWeights_2:float=0.0417) -> (Pose) {
    alphaBoolBlend {blendOption=Linear}
    boneToModify {boneName=dyn_necklace_charm_B_r_MSTR}
    constraintSetup [{targetBone {boneName=dyn_necklace_main_02}}, {targetBone {boneName=dyn_necklace_main_01}}, {targetBone {boneName=dyn_necklace_base}}]
    constraintWeights [1,1,1]
   }
   n24 = "Constraint - dyn_necklace_charm_A_r_MSTR (PPPP)"!(
    ComponentPose=@n23
    Alpha=1
    ConstraintWeights_0=0.0125
    ConstraintWeights_1:float=0.0345
    ConstraintWeights_2:float=0.3032
    ConstraintWeights_3:float=0.6498
   ) -> (Pose) {
    alphaBoolBlend {blendOption=Linear}
    boneToModify {boneName=dyn_necklace_charm_A_r_MSTR}
    constraintSetup [
     {targetBone {boneName=dyn_necklace_bk_01}}
     {targetBone {boneName=dyn_necklace_main_02}}
     {targetBone {boneName=dyn_necklace_main_01}}
     {targetBone {boneName=dyn_necklace_base}}
    ]
    constraintWeights [1,1,1,1]
   }
   n25 = Get L Head Rotation Alpha() -> (L Head Rotation Alpha:double)
   n26 = "Transform (Modify) Bone - Bone: dyn_necklace_charm_A_l_MSTR"(ComponentPose=@n24, Alpha=@n25, Translation="-1,0,0", Rotation="0,0,0", Scale="1,1,1") {
    rotationMode=BMM_Additive
    rotationSpace=BCS_BoneSpace
    translationMode=BMM_Additive
    translationSpace=BCS_BoneSpace
    alphaBoolBlend {blendOption=Linear}
    boneToModify {boneName=dyn_necklace_charm_A_l_MSTR}
    translation {x=-1}
   }
   n27 = "Transform (Modify) Bone - Bone: dyn_necklace_charm_B_l_MSTR"(ComponentPose=@n26, Alpha=@n25, Translation="-1,0,0", Rotation="0,0,0", Scale="1,1,1") {
    rotationMode=BMM_Additive
    rotationSpace=BCS_BoneSpace
    translationMode=BMM_Additive
    translationSpace=BCS_BoneSpace
    alphaBoolBlend {blendOption=Linear}
    boneToModify {boneName=dyn_necklace_charm_B_l_MSTR}
    translation {x=-1}
   }
   n28 = "Transform (Modify) Bone - Bone: dyn_necklace_charm_A_r_MSTR"(ComponentPose=@n27, Alpha=@n25, Translation="-0.5,0,0", Rotation="25,0,0", Scale="1,1,1") {
    rotationMode=BMM_Additive
    rotationSpace=BCS_BoneSpace
    translationMode=BMM_Additive
    translationSpace=BCS_BoneSpace
    alphaBoolBlend {blendOption=Linear}
    boneToModify {boneName=dyn_necklace_charm_A_r_MSTR}
    rotation {pitch=25}
    translation {x=-0.5}
   }
   Output Pose(Result=Component To Local(ComponentPose=@n28)) {layerGroup=None}
   Get Head Back Rotation Alpha()
  ]}
 EventGraph {nodes [
   Event BlueprintUpdateAnimation()
   > n1 = Sequence!(execute) -> (then_0, then_1, then_2:Exec, then_3:Exec)
   > n2 = Branch(Condition=OR Boolean!(A=Get bIsJumping(), B=Get bIsFalling(), C:bool=Get bIsCrouching()) -> (ReturnValue))
   > Set thigh_rot_r(thigh_rot_r=0)
   > Set thigh_rot_l(thigh_rot_l=0)
   n3 = BreakTransform(InTransform=GetDeltaTransformFromRefPose(self=GetOwnerSkeletalMeshComponent(), BoneName=thigh_l, BaseName=pelvis))
   > Set Thigh L Rot Crouch(Thigh L Rot Crouch:double=MapRangeClamped(Value=BreakRotator(InRot=@n3.Rotation).Yaw, InRangeA=-40, InRangeB=-100)) -> (Output_Get:double)
   n4 = BreakTransform(InTransform=GetDeltaTransformFromRefPose(self=GetOwnerSkeletalMeshComponent(), BoneName=thigh_r, BaseName=pelvis))
   > Set Thigh_R_Rot_Crouch(Thigh_R_Rot_Crouch:double=MapRangeClamped(Value=BreakRotator(InRot=@n4.Rotation).Yaw, InRangeA=-40, InRangeB=-100)) -> (Output_Get:double)
   Set Thigh L Rot Crouch(execute=@n2.else, Thigh L Rot Crouch=0)
   > Set Thigh_R_Rot_Crouch(Thigh_R_Rot_Crouch=0)
   n5 = Get Skel_Mesh()
   > Set thigh_rot_r(thigh_rot_r=GetAlphaFromReferenceTransformDelta(SkeletalMeshComponent=@n5, BoneName=thigh_r, BaseName=pelvis, InputRange=(X=0,Y=-90), TwistAxis="0,0,1"))
   > Set thigh_rot_l(thigh_rot_l=GetAlphaFromReferenceTransformDelta(SkeletalMeshComponent=@n5, BoneName=thigh_l, BaseName=pelvis, InputRange=(X=-20,Y=-101), TwistAxis="0,0,1"))
   n6 = Branch(execute=@n1.then_1, Condition=Get bIsSwimSprinting() -> (bIsSwimSprinting:bool))
   > Set gravity_override(gravity_override="0, 0, 0")
   n7 = Branch(execute=@n6.else, Condition=Get bIsSkyDiving())
   > n8 = Branch(Condition=Get bIsParachuteOpened())
   Set gravity_override(execute=[@n7.else, @n8.then], gravity_override="0, 0,-980")
   n9 = Branch(execute=@n8.else, Condition="float > float"(A=Get PawnVelocityZ(), B=50) {is=title K2Node_CallFunction})
   > Set gravity_override(gravity_override="0, 0,200")
   n10 = vector * float(
    A=GetForwardVector!(InRot:Rotator=GetSocketRotation!(self:Scene Component Object Reference=@n5, InSocketName:Name=pelvis) -> (ReturnValue:Rotator)) -> (ReturnValue:Vector)
    B=-1000
   )
   Set gravity_override(
    execute=@n9.else
    gravity_override=vector + vector(A:Vector=@n10, B:Vector="0, 0,-200") -> (ReturnValue:Vector) {is=title K2Node_CommutativeAssociativeBinaryOperator}
   )
   n11 = Branch(execute=@n1.then_3, Condition=Get bIsSkyDiving())
   > Set Skydive Alpha(Skydive Alpha:double=0.55) -> (Output_Get:double)
   Set Skydive Alpha(execute=@n11.else, Skydive Alpha=1)
   n12 = GetOwnerSkeletalMeshComponent()
   n13 = BreakRotator(InRot=BreakTransform(InTransform=GetDeltaTransformFromRefPose(self=@n12, BoneName=neck_01, BaseName=spine_05)).Rotation)
   Set L Head Rotation Alpha(execute=@n1.then_2, L Head Rotation Alpha:double=MapRangeClamped(Value=@n13, InRangeA=0, InRangeB=30)) -> (Output_Get:double)
   > Set Head Back Rotation Alpha(Head Back Rotation Alpha:double=MapRangeClamped(Value=@n13.Yaw, InRangeA=-2, InRangeB=-20)) -> (Output_Get:double)
   Event BlueprintInitializeAnimation()
   > Set Skel_Mesh(Skel_Mesh=GetOwnerSkeletalMeshComponent())
   n14 = GetSocketLocation!(self:Scene Component Object Reference=@n5, InSocketName:Name=pelvis) -> (ReturnValue:Vector)
   DrawDebugLine!(
    execute
    LineStart:Vector=@n14
    LineEnd:Vector=vector + vector(A=@n14, B=@n10)
    LineColor:Linear Color Structure=(R=1,G=0,B=0,A=1)
    Duration:float=0
    Thickness:float=3
    DepthPriority:EDrawDebugSceneDepthPriorityGroup Enum=World
   ) -> (then)
  ]}
]
variables [
 gravity_override {type=Vector}
 Head Back Rotation Alpha {type=number}
 L Head Rotation Alpha {type=number}
 Skel_Mesh {type=SkeletalMeshComponent}
 Skydive Alpha {type=number}
 Thigh L Rot Crouch {type=number}
 Thigh_R_Rot_Crouch {type=number}
 thigh_rot_l {type=number}
 thigh_rot_r {type=number}
]
