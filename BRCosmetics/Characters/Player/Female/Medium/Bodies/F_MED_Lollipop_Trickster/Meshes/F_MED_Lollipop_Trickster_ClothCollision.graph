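# Physics asset for the F_MED_Lollipop_Trickster female body cosmetic, named as its cloth collision proxy.
# Single capsules cover each of the five spine bones, and each upper arm carries two overlapping, slightly offset capsules, mirrored left to right.
# Constraints chain the spine bones in sequence and attach both upper arms to the top spine bone, with all swing and twist motion set to free.
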PhysicsAsset /BRCosmetics/Characters/Player/Female/Medium/Bodies/F_MED_Lollipop_Trickster/Meshes/F_MED_Lollipop_Trickster_ClothCollision
bodies {
 spine_01 [{shapeType=Capsule, length=10, radius=5}]
 spine_02 [{shapeType=Capsule, length=6.375, radius=8.625, center=<0 1.393 0.0002>}]
 spine_03 [{shapeType=Capsule, length=6.375, radius=8.625, center=<0 2.3817 0>}]
 spine_04 [{shapeType=Capsule, length=6.375, radius=8.625, center=<0 2.3817 0>}]
 spine_05 [{shapeType=Capsule, length=10, radius=5}]
 upperarm_l [
  {shapeType=Capsule, length=16.0978, radius=3.1714, center=<11.3725 0.5282 -0.5517>, rotation=<rot 90 -90 -86.002>}
  {shapeType=Capsule, length=16.0978, radius=3.1714, center=<11.0113 -1.3434 -0.5516>, rotation=<rot 90 0 -2.904>}
 ]
 upperarm_r [
  {shapeType=Capsule, length=16.0978, radius=3.1714, center=<-11.3725 -0.5282 0.5517>, rotation=<rot 90 -90 -86.002>}
  {shapeType=Capsule, length=16.0978, radius=3.1714, center=<-11.0113 1.3434 0.5516>, rotation=<rot 90 0 -2.904>}
 ]
}
constraints [
 {
  bone1Name=spine_02
  bone2Name=spine_01
  swing1LimitDegrees=45
  swing1Motion=Free
  swing2LimitDegrees=45
  swing2Motion=Free
  twistLimitDegrees=45
  twistMotion=Free
 }
 {
  bone1Name=spine_03
  bone2Name=spine_02
  swing1LimitDegrees=45
  swing1Motion=Free
  swing2LimitDegrees=45
  swing2Motion=Free
  twistLimitDegrees=45
  twistMotion=Free
 }
 {
  bone1Name=spine_04
  bone2Name=spine_03
  swing1LimitDegrees=45
  swing1Motion=Free
  swing2LimitDegrees=45
  swing2Motion=Free
  twistLimitDegrees=45
  twistMotion=Free
 }
 {
  bone1Name=spine_05
  bone2Name=spine_04
  swing1LimitDegrees=45
  swing1Motion=Free
  swing2LimitDegrees=45
  swing2Motion=Free
  twistLimitDegrees=45
  twistMotion=Free
 }
 {
  bone1Name=upperarm_l
  bone2Name=spine_05
  swing1LimitDegrees=45
  swing1Motion=Free
  swing2LimitDegrees=45
  swing2Motion=Free
  twistLimitDegrees=45
  twistMotion=Free
 }
 {
  bone1Name=upperarm_r
  bone2Name=spine_05
  swing1LimitDegrees=45
  swing1Motion=Free
  swing2LimitDegrees=45
  swing2Motion=Free
  twistLimitDegrees=45
  twistMotion=Free
 }
]
