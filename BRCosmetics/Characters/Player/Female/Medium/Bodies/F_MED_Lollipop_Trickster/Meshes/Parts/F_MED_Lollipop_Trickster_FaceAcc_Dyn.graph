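# Physics asset for the Lollipop Trickster female medium body's face accessory part, covering dynamic bangs, left and right ponytails, and earrings.
# Capsules on each dynamic bone form constrained chains rooted at the head, whose sphere and capsule join a neck capsule through a free-rotation constraint.
# Ponytails have tight root joints, very wide swing at the second joint, then branching segments with moderate swing.
# Bangs and earrings use moderate swing and tight twist limits.
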PhysicsAsset /BRCosmetics/Characters/Player/Female/Medium/Bodies/F_MED_Lollipop_Trickster/Meshes/Parts/F_MED_Lollipop_Trickster_FaceAcc_Dyn
bodies {
 dyn_bangs_long_bk_r_01 [{shapeType=Capsule, length=3, radius=1.2, center=<1.5 0 0>, rotation=<rot 90 0 0>}]
 dyn_bangs_long_bk_r_02 [{shapeType=Capsule, length=3, radius=1.2, center=<1.5 0 0>, rotation=<rot 90 0 0>}]
 dyn_bangs_long_fr_r_01 [{shapeType=Capsule, length=3, radius=1.2, center=<1.5 0 0>, rotation=<rot 90 0 0>}]
 dyn_bangs_long_fr_r_02 [{shapeType=Capsule, length=3, radius=1.2, center=<1.5 0 0>, rotation=<rot 90 0 0>}]
 dyn_bangs_long_l_01 [{shapeType=Capsule, length=3.4375, radius=0.7625, center=<1.5 0 0>, rotation=<rot 90 0 0>}]
 dyn_bangs_long_l_02 [{shapeType=Capsule, length=3.5312, radius=0.6687, center=<1.5 0 0>, rotation=<rot 90 0 0>}]
 dyn_bangs_long_l_03 [{shapeType=Capsule, length=3.7, radius=0.5, center=<1.5 0 0>, rotation=<rot 90 0 0>}]
 dyn_bangs_long_r_01 [{shapeType=Capsule, length=3.0625, radius=1.1375, center=<1.5 0 0>, rotation=<rot 90 0 0>}]
 dyn_bangs_long_r_02 [{shapeType=Capsule, length=3.5625, radius=0.6375, center=<1.5 0 0>, rotation=<rot 90 0 0>}]
 dyn_bangs_long_r_03 [{shapeType=Capsule, length=3.6063, radius=0.5938, center=<1.5 0 0>, rotation=<rot 90 0 0>}]
 dyn_bangs_short_l_01 [{shapeType=Capsule, length=3, radius=1.2, center=<1.5 0 0>, rotation=<rot 90 0 0>}]
 dyn_bangs_short_l_02 [{shapeType=Capsule, length=3, radius=1.2, center=<1.5 0 0>, rotation=<rot 90 0 0>}]
 dyn_bangs_short_l_03 [{shapeType=Capsule, length=3, radius=1.2, center=<1.5 0 0>, rotation=<rot 90 0 0>}]
 dyn_bangs_short_r [{shapeType=Capsule, length=3, radius=1.2, center=<1.5 0 0>, rotation=<rot 90 0 0>}]
 dyn_earring_l_01 [{shapeType=Capsule, length=0.5312, radius=0.5437, center=<0.5 0 0>, rotation=<rot 90 0 0>}]
 dyn_earring_l_02 [{shapeType=Capsule, length=0.5312, radius=0.9187, center=<1.1161 0 0>, rotation=<rot 90 0 0>}]
 dyn_earring_r_01 [{shapeType=Capsule, length=0.5312, radius=0.5437, center=<0.5 0 0>, rotation=<rot 90 0 0>}]
 dyn_earring_r_02 [{shapeType=Capsule, length=0.5312, radius=0.9187, center=<1.1161 0 0>, rotation=<rot 90 0 0>}]
 dyn_ponytail_l_01 [{shapeType=Capsule, length=3, radius=2, center=<1.5 0 0>, rotation=<rot 90 0 0>}]
 dyn_ponytail_l_02 [{shapeType=Capsule, length=3, radius=2, center=<4.0514 -0.0001 0>, rotation=<rot 90 0 0>}]
 dyn_ponytail_l_03 [{shapeType=Capsule, length=3, radius=2, center=<1.5 0 0>, rotation=<rot 90 0 0>}]
 dyn_ponytail_l_04 [{shapeType=Capsule, length=3, radius=2, center=<1.5 0 0>, rotation=<rot 90 0 0>}]
 dyn_ponytail_l_05 [{shapeType=Capsule, length=3, radius=2, center=<1.5 0 0>, rotation=<rot 90 0 0>}]
 dyn_ponytail_l_06 [{shapeType=Capsule, length=3, radius=2, center=<1.5 0 0>, rotation=<rot 90 0 0>}]
 dyn_ponytail_l_07 [{shapeType=Capsule, length=3, radius=2, center=<1.5 0 0>, rotation=<rot 90 0 0>}]
 dyn_ponytail_r_01 [{shapeType=Capsule, length=3, radius=2, center=<1.5 0 0>, rotation=<rot 90 0 0>}]
 dyn_ponytail_r_02 [{shapeType=Capsule, length=3, radius=2, center=<1.5 0 0>, rotation=<rot 90 0 0>}]
 dyn_ponytail_r_03 [{shapeType=Capsule, length=3, radius=2, center=<1.5 0 0>, rotation=<rot 90 0 0>}]
 dyn_ponytail_r_04 [{shapeType=Capsule, length=3, radius=2, center=<1.5 0 0>, rotation=<rot 90 0 0>}]
 dyn_ponytail_r_05 [{shapeType=Capsule, length=3, radius=2, center=<1.5 0 0>, rotation=<rot 90 0 0>}]
 dyn_ponytail_r_06 [{shapeType=Capsule, length=3, radius=2, center=<1.5 0 0>, rotation=<rot 90 0 0>}]
 dyn_ponytail_r_07 [{shapeType=Capsule, length=3, radius=2, center=<1.5 0 0>, rotation=<rot 90 0 0>}]
 head [
  {shapeType=Sphere, radius=7.697, center=<6.4024 -0.5133 0>}
  {shapeType=Capsule, length=6.5625, radius=6.4375, center=<2.675 3.1784 0>, rotation=<rot 90 0 0>}
 ]
 neck_02 [{shapeType=Capsule, length=8, radius=5, rotation=<rot 90 0 0>}]
}
constraints [
 {
  bone1Name=head
  bone2Name=neck_02
  swing1LimitDegrees=45
  swing1Motion=Free
  swing2LimitDegrees=45
  swing2Motion=Free
  twistLimitDegrees=45
  twistMotion=Free
 }
 {bone1Name=dyn_ponytail_l_01, bone2Name=head, swing1LimitDegrees=5, swing2LimitDegrees=5, twistLimitDegrees=5}
 {bone1Name=dyn_ponytail_l_02, bone2Name=dyn_ponytail_l_01, swing1LimitDegrees=120, swing2LimitDegrees=120, twistLimitDegrees=5}
 {bone1Name=dyn_ponytail_l_03, bone2Name=dyn_ponytail_l_02, swing1LimitDegrees=25, swing2LimitDegrees=25, twistLimitDegrees=5}
 {bone1Name=dyn_ponytail_l_05, bone2Name=dyn_ponytail_l_03, swing1LimitDegrees=25, swing2LimitDegrees=25, twistLimitDegrees=5}
 {bone1Name=dyn_ponytail_l_06, bone2Name=dyn_ponytail_l_03, swing1LimitDegrees=25, swing2LimitDegrees=25, twistLimitDegrees=5}
 {bone1Name=dyn_ponytail_l_07, bone2Name=dyn_ponytail_l_03, swing1LimitDegrees=25, swing2LimitDegrees=25, twistLimitDegrees=5}
 {bone1Name=dyn_ponytail_l_04, bone2Name=dyn_ponytail_l_02, swing1LimitDegrees=25, swing2LimitDegrees=25, twistLimitDegrees=5}
 {bone1Name=dyn_ponytail_r_01, bone2Name=head, swing1LimitDegrees=5, swing2LimitDegrees=5, twistLimitDegrees=5}
 {bone1Name=dyn_ponytail_r_02, bone2Name=dyn_ponytail_r_01, swing1LimitDegrees=120, swing2LimitDegrees=120, twistLimitDegrees=5}
 {bone1Name=dyn_ponytail_r_04, bone2Name=dyn_ponytail_r_02, swing1LimitDegrees=25, swing2LimitDegrees=25, twistLimitDegrees=5}
 {bone1Name=dyn_ponytail_r_03, bone2Name=dyn_ponytail_r_02, swing1LimitDegrees=25, swing2LimitDegrees=25, twistLimitDegrees=5}
 {bone1Name=dyn_ponytail_r_05, bone2Name=dyn_ponytail_r_03, swing1LimitDegrees=25, swing2LimitDegrees=25, twistLimitDegrees=5}
 {bone1Name=dyn_ponytail_r_06, bone2Name=dyn_ponytail_r_03, swing1LimitDegrees=25, swing2LimitDegrees=25, twistLimitDegrees=5}
 {bone1Name=dyn_ponytail_r_07, bone2Name=dyn_ponytail_r_02, swing1LimitDegrees=25, swing2LimitDegrees=25, twistLimitDegrees=5}
 {bone1Name=dyn_bangs_short_r, bone2Name=head, swing1LimitDegrees=25, swing2LimitDegrees=25, twistLimitDegrees=5}
 {bone1Name=dyn_bangs_short_l_01, bone2Name=head, swing1LimitDegrees=25, swing2LimitDegrees=25, twistLimitDegrees=5}
 {bone1Name=dyn_bangs_short_l_02, bone2Name=head, swing1LimitDegrees=25, swing2LimitDegrees=25, twistLimitDegrees=5}
 {bone1Name=dyn_bangs_short_l_03, bone2Name=head, swing1LimitDegrees=25, swing2LimitDegrees=25, twistLimitDegrees=5}
 {bone1Name=dyn_bangs_long_l_01, bone2Name=head, swing1LimitDegrees=25, swing2LimitDegrees=25, twistLimitDegrees=5}
 {bone1Name=dyn_bangs_long_l_02, bone2Name=dyn_bangs_long_l_01, swing1LimitDegrees=25, swing2LimitDegrees=25, twistLimitDegrees=5}
 {bone1Name=dyn_bangs_long_l_03, bone2Name=dyn_bangs_long_l_02, swing1LimitDegrees=25, swing2LimitDegrees=25, twistLimitDegrees=5}
 {bone1Name=dyn_bangs_long_r_01, bone2Name=head, swing1LimitDegrees=25, swing2LimitDegrees=25, twistLimitDegrees=5}
 {bone1Name=dyn_bangs_long_r_02, bone2Name=dyn_bangs_long_r_01, swing1LimitDegrees=25, swing2LimitDegrees=25, twistLimitDegrees=5}
 {bone1Name=dyn_bangs_long_r_03, bone2Name=dyn_bangs_long_r_02, swing1LimitDegrees=25, swing2LimitDegrees=25, twistLimitDegrees=5}
 {bone1Name=dyn_bangs_long_bk_r_01, bone2Name=head, swing1LimitDegrees=25, swing2LimitDegrees=25, twistLimitDegrees=5}
 {bone1Name=dyn_bangs_long_bk_r_02, bone2Name=dyn_bangs_long_bk_r_01, swing1LimitDegrees=25, swing2LimitDegrees=25, twistLimitDegrees=5}
 {bone1Name=dyn_bangs_long_fr_r_01, bone2Name=head, swing1LimitDegrees=25, swing2LimitDegrees=25, twistLimitDegrees=5}
 {bone1Name=dyn_bangs_long_fr_r_02, bone2Name=dyn_bangs_long_fr_r_01, swing1LimitDegrees=25, swing2LimitDegrees=25, twistLimitDegrees=5}
 {bone1Name=dyn_earring_l_01, bone2Name=head, swing1LimitDegrees=25, swing2LimitDegrees=25, twistLimitDegrees=5}
 {bone1Name=dyn_earring_l_02, bone2Name=dyn_earring_l_01, swing1LimitDegrees=25, swing2LimitDegrees=25, twistLimitDegrees=5}
 {bone1Name=dyn_earring_r_01, bone2Name=head, swing1LimitDegrees=25, swing2LimitDegrees=25, twistLimitDegrees=5}
 {bone1Name=dyn_earring_r_02, bone2Name=dyn_earring_r_01, swing1LimitDegrees=25, swing2LimitDegrees=25, twistLimitDegrees=5}
]
