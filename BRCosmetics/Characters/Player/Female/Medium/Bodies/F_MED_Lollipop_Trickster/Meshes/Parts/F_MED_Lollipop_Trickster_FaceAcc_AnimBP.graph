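# Anim Blueprint for the Lollipop Trickster female face-accessory mesh, adding secondary physics motion to bangs and earrings.
# It copies the attached parent's pose, then chains per-bone AnimDynamics nodes: lightly blended angular springs for bangs, damped swing for earrings.
# A head-space RigidBody node follows, using the FaceAcc_Dyn physics asset.
# Its gravity is normal except during skydiving before parachute opening, when it weakens and points up or down by vertical velocity.
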AnimBlueprint /BRCosmetics/Characters/Player/Female/Medium/Bodies/F_MED_Lollipop_Trickster/Meshes/Parts/F_MED_Lollipop_Trickster_FaceAcc_AnimBP
target_skeleton=&/F_MED_Lollipop_Trickster_FaceAcc_Skeleton
graphs [
 AnimGraph {nodes [
   n1 = Anim Dynamics - Bone: dyn_bangs_short_r(ComponentPose=Local To Component(LocalPose=Copy Pose From Mesh() {bUseAttachedParent=true}), Alpha=1, ExternalForce="0,0,0") {
    alpha=0.1
    angularDampingOverride=1
    angularSpringConstant=1500
    bAngularSpring=true
    bGravityOverrideInSimSpace=true
    bOverrideAngularDamping=true
    gravityScale=0.35
    simulationSpace=RootRelative
    alphaBoolBlend {blendOption=Linear}
    boundBone {boneName=dyn_bangs_short_r}
    componentLinearAccScale=<0.2 0.2 0.2>
    componentLinearVelScale=<0.2 0.2 0.2>
    physicsBodyDefinitions [{
      sphereCollisionRadius=0
      boundBone {boneName=dyn_bangs_short_r}
      boxExtents=<10 20 20>
      constraintSetup {angularLimitsMax=<0 0 7>, angularLimitsMin=<0 -15 -7>, angularTarget=<0 0 0>}
      localJointOffset=<10 0 0>
     }]
   }
   n2 = Anim Dynamics - Bone: dyn_bangs_short_l_01(ComponentPose=@n1, Alpha=1, ExternalForce="0,0,0") {
    alpha=0.3
    angularDampingOverride=1
    angularSpringConstant=1500
    bAngularSpring=true
    bGravityOverrideInSimSpace=true
    bOverrideAngularDamping=true
    gravityScale=0.35
    simulationSpace=RootRelative
    alphaBoolBlend {blendOption=Linear}
    boundBone {boneName=dyn_bangs_short_l_01}
    componentLinearAccScale=<0.2 0.2 0.2>
    componentLinearVelScale=<0.2 0.2 0.2>
    physicsBodyDefinitions [{
      sphereCollisionRadius=0
      boundBone {boneName=dyn_bangs_short_l_01}
      boxExtents=<10 20 20>
      constraintSetup {angularLimitsMax=<0 0 7>, angularLimitsMin=<0 -15 -7>, angularTarget=<0 0 0>}
      localJointOffset=<10 0 0>
     }]
   }
   n3 = Anim Dynamics - Bone: dyn_bangs_short_l_02(ComponentPose=@n2, Alpha=1, ExternalForce="0,0,0") {
    alpha=0.4
    angularDampingOverride=1
    angularSpringConstant=1500
    bAngularSpring=true
    bGravityOverrideInSimSpace=true
    bOverrideAngularDamping=true
    gravityScale=0.35
    simulationSpace=RootRelative
    alphaBoolBlend {blendOption=Linear}
    boundBone {boneName=dyn_bangs_short_l_02}
    componentLinearAccScale=<0.2 0.2 0.2>
    componentLinearVelScale=<0.2 0.2 0.2>
    physicsBodyDefinitions [{
      sphereCollisionRadius=0
      boundBone {boneName=dyn_bangs_short_l_02}
      boxExtents=<10 20 20>
      constraintSetup {angularLimitsMax=<0 3 15>, angularLimitsMin=<0 -20 -15>, angularTarget=<0 0 0>}
      localJointOffset=<10 0 0>
     }]
   }
   n4 = Anim Dynamics - Bone: dyn_bangs_short_l_03(ComponentPose=@n3, Alpha=1, ExternalForce="0,0,0") {
    alpha=0.1
    angularDampingOverride=1
    angularSpringConstant=1500
    bAngularSpring=true
    bGravityOverrideInSimSpace=true
    bOverrideAngularDamping=true
    gravityScale=0.35
    simulationSpace=RootRelative
    alphaBoolBlend {blendOption=Linear}
    boundBone {boneName=dyn_bangs_short_l_03}
    componentLinearAccScale=<0.2 0.2 0.2>
    componentLinearVelScale=<0.2 0.2 0.2>
    physicsBodyDefinitions [{
      sphereCollisionRadius=0
      boundBone {boneName=dyn_bangs_short_l_03}
      boxExtents=<10 20 20>
      constraintSetup {angularLimitsMax=<0 3 15>, angularLimitsMin=<0 -20 -15>, angularTarget=<0 0 0>}
      localJointOffset=<10 0 0>
     }]
   }
   n5 = Anim Dynamics - Bone: dyn_bangs_long_bk_r_01(ComponentPose=@n4, Alpha=1, ExternalForce="0,0,0") {
    alpha=0.1
    angularDampingOverride=1
    angularSpringConstant=1500
    bAngularSpring=true
    bGravityOverrideInSimSpace=true
    bOverrideAngularDamping=true
    gravityScale=0.35
    simulationSpace=RootRelative
    alphaBoolBlend {blendOption=Linear}
    boundBone {boneName=dyn_bangs_long_bk_r_01}
    componentLinearAccScale=<0.2 0.2 0.2>
    componentLinearVelScale=<0.2 0.2 0.2>
    physicsBodyDefinitions [{
      sphereCollisionRadius=0
      boundBone {boneName=dyn_bangs_long_bk_r_01}
      boxExtents=<10 20 20>
      constraintSetup {angularLimitsMax=<2 0 10>, angularLimitsMin=<-2 -5 -5>, angularTarget=<0 0 0>}
      localJointOffset=<10 0 0>
     }]
    relativeSpaceBone {boneName=head}
   }
   n6 = Anim Dynamics - Bone: dyn_bangs_long_bk_r_02(ComponentPose=@n5, Alpha=1, ExternalForce="0,0,0") {
    alpha=0.1
    angularDampingOverride=1
    angularSpringConstant=1500
    bAngularSpring=true
    bGravityOverrideInSimSpace=true
    bOverrideAngularDamping=true
    gravityScale=0.35
    simulationSpace=RootRelative
    alphaBoolBlend {blendOption=Linear}
    boundBone {boneName=dyn_bangs_long_bk_r_02}
    componentLinearAccScale=<0.2 0.2 0.2>
    componentLinearVelScale=<0.2 0.2 0.2>
    physicsBodyDefinitions [{
      sphereCollisionRadius=0
      boundBone {boneName=dyn_bangs_long_bk_r_02}
      boxExtents=<10 20 20>
      constraintSetup {angularLimitsMax=<3 1 10>, angularLimitsMin=<-3 -5 -5>, angularTarget=<0 0 0>}
      localJointOffset=<10 0 0>
     }]
    relativeSpaceBone {boneName=dyn_bangs_long_l_01}
   }
   n7 = Anim Dynamics - Bone: dyn_bangs_long_l_01(ComponentPose=@n6, Alpha=1, ExternalForce="0,0,0") {
    alpha=0.1
    angularDampingOverride=1
    angularSpringConstant=1500
    bAngularSpring=true
    bGravityOverrideInSimSpace=true
    bOverrideAngularDamping=true
    gravityScale=0.35
    simulationSpace=RootRelative
    alphaBoolBlend {blendOption=Linear}
    boundBone {boneName=dyn_bangs_long_l_01}
    componentLinearAccScale=<0.2 0.2 0.2>
    componentLinearVelScale=<0.2 0.2 0.2>
    physicsBodyDefinitions [{
      sphereCollisionRadius=0
      boundBone {boneName=dyn_bangs_long_l_01}
      boxExtents=<10 20 20>
      constraintSetup {angularLimitsMax=<2 0 5>, angularLimitsMin=<-2 -5 -10>, angularTarget=<0 0 0>}
      localJointOffset=<10 0 0>
     }]
    relativeSpaceBone {boneName=head}
   }
   n8 = Anim Dynamics - Bone: dyn_bangs_long_l_02(ComponentPose=@n7, Alpha=1, ExternalForce="0,0,0") {
    alpha=0.1
    angularDampingOverride=1
    angularSpringConstant=1500
    bAngularSpring=true
    bGravityOverrideInSimSpace=true
    bOverrideAngularDamping=true
    gravityScale=0.35
    simulationSpace=RootRelative
    alphaBoolBlend {blendOption=Linear}
    boundBone {boneName=dyn_bangs_long_l_02}
    componentLinearAccScale=<0.2 0.2 0.2>
    componentLinearVelScale=<0.2 0.2 0.2>
    physicsBodyDefinitions [{
      sphereCollisionRadius=0
      boundBone {boneName=dyn_bangs_long_l_02}
      boxExtents=<10 20 20>
      constraintSetup {angularLimitsMax=<3 1 5>, angularLimitsMin=<-3 -10 -20>, angularTarget=<0 0 0>}
      localJointOffset=<10 0 0>
     }]
    relativeSpaceBone {boneName=dyn_bangs_long_l_01}
   }
   n9 = Anim Dynamics - Bone: dyn_bangs_long_l_03(ComponentPose=@n8, Alpha=1, ExternalForce="0,0,0") {
    alpha=0.1
    angularDampingOverride=1
    angularSpringConstant=1500
    bAngularSpring=true
    bGravityOverrideInSimSpace=true
    bOverrideAngularDamping=true
    gravityScale=0.35
    simulationSpace=RootRelative
    alphaBoolBlend {blendOption=Linear}
    boundBone {boneName=dyn_bangs_long_l_03}
    componentLinearAccScale=<0.2 0.2 0.2>
    componentLinearVelScale=<0.2 0.2 0.2>
    physicsBodyDefinitions [{
      sphereCollisionRadius=0
      boundBone {boneName=dyn_bangs_long_l_03}
      boxExtents=<10 20 20>
      constraintSetup {angularLimitsMax=<3 10 25>, angularLimitsMin=<-3 -10 -25>, angularTarget=<0 0 0>}
      localJointOffset=<10 0 0>
     }]
    relativeSpaceBone {boneName=dyn_bangs_long_l_02}
   }
   n10 = Anim Dynamics - Bone: dyn_bangs_long_r_01(ComponentPose=@n9, Alpha=1, ExternalForce="0,0,0") {
    alpha=0.1
    angularDampingOverride=1
    angularSpringConstant=1500
    bAngularSpring=true
    bGravityOverrideInSimSpace=true
    bOverrideAngularDamping=true
    gravityScale=0.35
    simulationSpace=RootRelative
    alphaBoolBlend {blendOption=Linear}
    boundBone {boneName=dyn_bangs_long_r_01}
    componentLinearAccScale=<0.2 0.2 0.2>
    componentLinearVelScale=<0.2 0.2 0.2>
    physicsBodyDefinitions [{
      sphereCollisionRadius=0
      boundBone {boneName=dyn_bangs_long_r_01}
      boxExtents=<10 20 20>
      constraintSetup {angularLimitsMax=<2 0 10>, angularLimitsMin=<-2 -5 -10>, angularTarget=<0 0 0>}
      localJointOffset=<10 0 0>
     }]
    relativeSpaceBone {boneName=head}
   }
   n11 = Anim Dynamics - Bone: dyn_bangs_long_r_02(ComponentPose=@n10, Alpha=1, ExternalForce="0,0,0") {
    alpha=0.1
    angularDampingOverride=1
    angularSpringConstant=1500
    bAngularSpring=true
    bGravityOverrideInSimSpace=true
    bOverrideAngularDamping=true
    gravityScale=0.35
    simulationSpace=RootRelative
    alphaBoolBlend {blendOption=Linear}
    boundBone {boneName=dyn_bangs_long_r_02}
    componentLinearAccScale=<0.2 0.2 0.2>
    componentLinearVelScale=<0.2 0.2 0.2>
    physicsBodyDefinitions [{
      sphereCollisionRadius=0
      boundBone {boneName=dyn_bangs_long_r_02}
      boxExtents=<10 20 20>
      constraintSetup {angularLimitsMax=<3 1 20>, angularLimitsMin=<-3 -10 -20>, angularTarget=<0 0 0>}
      localJointOffset=<10 0 0>
     }]
    relativeSpaceBone {boneName=dyn_bangs_long_l_01}
   }
   n12 = Anim Dynamics - Bone: dyn_bangs_long_r_03(ComponentPose=@n11, Alpha=1, ExternalForce="0,0,0") {
    alpha=0.1
    angularDampingOverride=1
    angularSpringConstant=1500
    bAngularSpring=true
    bGravityOverrideInSimSpace=true
    bOverrideAngularDamping=true
    gravityScale=0.35
    simulationSpace=RootRelative
    alphaBoolBlend {blendOption=Linear}
    boundBone {boneName=dyn_bangs_long_r_03}
    componentLinearAccScale=<0.2 0.2 0.2>
    componentLinearVelScale=<0.2 0.2 0.2>
    physicsBodyDefinitions [{
      sphereCollisionRadius=0
      boundBone {boneName=dyn_bangs_long_r_03}
      boxExtents=<10 20 20>
      constraintSetup {angularLimitsMax=<3 10 25>, angularLimitsMin=<-3 -10 -25>, angularTarget=<0 0 0>}
      localJointOffset=<10 0 0>
     }]
    relativeSpaceBone {boneName=dyn_bangs_long_r_03}
   }
   n13 = Anim Dynamics - Bone: dyn_bangs_long_fr_r_01(ComponentPose=@n12, Alpha=1, ExternalForce="0,0,0") {
    alpha=0.1
    angularDampingOverride=1
    angularSpringConstant=1500
    bAngularSpring=true
    bGravityOverrideInSimSpace=true
    bOverrideAngularDamping=true
    gravityScale=0.35
    simulationSpace=RootRelative
    alphaBoolBlend {blendOption=Linear}
    boundBone {boneName=dyn_bangs_long_fr_r_01}
    componentLinearAccScale=<0.2 0.2 0.2>
    componentLinearVelScale=<0.2 0.2 0.2>
    physicsBodyDefinitions [{
      sphereCollisionRadius=0
      boundBone {boneName=dyn_bangs_long_fr_r_01}
      boxExtents=<12 26 26>
      constraintSetup {angularLimitsMax=<2 0 12>, angularLimitsMin=<-2 -5 -12>, angularTarget=<0 0 0>}
      localJointOffset=<10 0 0>
     }]
    relativeSpaceBone {boneName=head}
   }
   n14 = Anim Dynamics - Bone: dyn_bangs_long_fr_r_02(ComponentPose=@n13, Alpha=1, ExternalForce="0,0,0") {
    alpha=0.1
    angularDampingOverride=1
    angularSpringConstant=1500
    bAngularSpring=true
    bGravityOverrideInSimSpace=true
    bOverrideAngularDamping=true
    gravityScale=0.35
    simulationSpace=RootRelative
    alphaBoolBlend {blendOption=Linear}
    boundBone {boneName=dyn_bangs_long_fr_r_02}
    componentLinearAccScale=<0.2 0.2 0.2>
    componentLinearVelScale=<0.2 0.2 0.2>
    physicsBodyDefinitions [{
      sphereCollisionRadius=0
      boundBone {boneName=dyn_bangs_long_fr_r_02}
      boxExtents=<12 26 26>
      constraintSetup {angularLimitsMax=<3 1 12>, angularLimitsMin=<-3 -5 -12>, angularTarget=<0 0 0>}
      localJointOffset=<10 0 0>
     }]
    relativeSpaceBone {boneName=dyn_bangs_long_l_01}
   }
   n15 = Anim Dynamics - Bone: dyn_earring_l_01(ComponentPose=@n14, Alpha=1, ExternalForce="0,0,0") {
    angularDampingOverride=0.998
    bGravityOverrideInSimSpace=true
    bOverrideAngularDamping=true
    bUseGravityOverride=true
    simulationSpace=RootRelative
    alphaBoolBlend {blendOption=Linear}
    boundBone {boneName=dyn_earring_l_01}
    componentAppliedLinearAccClamp=<3000 3000 3000>
    componentLinearAccScale=<0.3 0.3 0.3>
    componentLinearVelScale=<0.3 0.3 0.3>
    gravityOverride {z=-980}
    physicsBodyDefinitions [{
      sphereCollisionRadius=0
      boundBone {boneName=dyn_earring_l_01}
      boxExtents=<7 7 3>
      constraintSetup {angularLimitsMax=<2 3 10>, angularLimitsMin=<-2 -6 -10>, angularTarget=<0 0 0>}
      localJointOffset=<1 0 0>
     }]
    relativeSpaceBone {boneName=head}
   }
   n16 = Anim Dynamics - Bone: dyn_earring_l_02(ComponentPose=@n15, Alpha=1, ExternalForce="0,0,0") {
    angularDampingOverride=0.998
    bGravityOverrideInSimSpace=true
    bOverrideAngularDamping=true
    bUseGravityOverride=true
    simulationSpace=RootRelative
    alphaBoolBlend {blendOption=Linear}
    boundBone {boneName=dyn_earring_l_02}
    componentAppliedLinearAccClamp=<3000 3000 3000>
    componentLinearAccScale=<0.3 0.3 0.3>
    componentLinearVelScale=<0.3 0.3 0.3>
    gravityOverride {z=-980}
    physicsBodyDefinitions [{
      sphereCollisionRadius=0
      boundBone {boneName=dyn_earring_l_02}
      boxExtents=<7 7 3>
      constraintSetup {angularLimitsMax=<2 10 45>, angularLimitsMin=<-2 -45 -45>, angularTarget=<0 0 0>}
      localJointOffset=<1 0 0>
     }]
    relativeSpaceBone {boneName=dyn_earring_l_01}
   }
   n17 = Anim Dynamics - Bone: dyn_earring_r_01(ComponentPose=@n16, Alpha=1, ExternalForce="0,0,0") {
    angularDampingOverride=0.998
    bGravityOverrideInSimSpace=true
    bOverrideAngularDamping=true
    bUseGravityOverride=true
    simulationSpace=RootRelative
    alphaBoolBlend {blendOption=Linear}
    boundBone {boneName=dyn_earring_r_01}
    componentAppliedLinearAccClamp=<3000 3000 3000>
    componentLinearAccScale=<0.3 0.3 0.3>
    componentLinearVelScale=<0.3 0.3 0.3>
    gravityOverride {z=-980}
    physicsBodyDefinitions [{
      sphereCollisionRadius=0
      boundBone {boneName=dyn_earring_r_01}
      boxExtents=<7 7 3>
      constraintSetup {angularLimitsMax=<2 6 10>, angularLimitsMin=<-2 -3 -10>, angularTarget=<0 0 0>}
      localJointOffset=<1 0 0>
     }]
    relativeSpaceBone {boneName=head}
   }
   n18 = Anim Dynamics - Bone: dyn_earring_r_02(ComponentPose=@n17, Alpha=1, ExternalForce="0,0,0") {
    angularDampingOverride=0.998
    bGravityOverrideInSimSpace=true
    bOverrideAngularDamping=true
    bUseGravityOverride=true
    simulationSpace=RootRelative
    alphaBoolBlend {blendOption=Linear}
    boundBone {boneName=dyn_earring_r_02}
    componentAppliedLinearAccClamp=<3000 3000 3000>
    componentLinearAccScale=<0.3 0.3 0.3>
    componentLinearVelScale=<0.3 0.3 0.3>
    gravityOverride {z=-980}
    physicsBodyDefinitions [{
      sphereCollisionRadius=0
      boundBone {boneName=dyn_earring_r_02}
      boxExtents=<7 7 3>
      constraintSetup {angularLimitsMax=<2 45 45>, angularLimitsMin=<-2 -10 -45>, angularTarget=<0 0 0>}
      localJointOffset=<1 0 0>
     }]
    relativeSpaceBone {boneName=dyn_earring_r_01}
   }
   n19 = RigidBody!(ComponentPose=@n18, Alpha, OverrideWorldGravity:Vector=Get gravity_override() -> (gravity_override:Vector), ExternalForce) -> (Pose) {
    bClampLinearTranslationLimitToRefPose=true
    bOverrideWorldGravity=true
    overridePhysicsAsset=&/F_MED_Lollipop_Trickster_FaceAcc_Dyn
    simulationSpace=BaseBoneSpace
    alphaBoolBlend {blendOption=Linear}
    baseBoneRef {boneName=head}
    componentAppliedLinearAccClamp=<3000 3000 3000>
    componentLinearAccScale=<0.4 0.3 0.3>
    componentLinearVelScale=<0.4 0.3 0.3>
    overrideWorldGravity {z=-980}
   }
   Output Pose(Result=Component To Local(ComponentPose=@n19)) {layerGroup=None}
  ]}
 EventGraph {nodes [
   Event BlueprintUpdateAnimation()
   > n1 = Branch(Condition=Get bIsSkyDiving())
   > n2 = Branch(Condition=Get bIsParachuteOpened())
   Set gravity_override(execute=[@n1.else, @n2.then], gravity_override="0, 0,-980")
   n3 = Branch(execute=@n2.else, Condition="float > float"(A=Get PawnVelocityZ(), B=50) {is=title K2Node_CallFunction})
   > Set gravity_override(gravity_override="0, 0,200")
   Set gravity_override(execute=@n3.else, gravity_override="0, 0,-200")
   TryGetPawnOwner()
  ]}
]
variables [gravity_override {type=Vector}]
